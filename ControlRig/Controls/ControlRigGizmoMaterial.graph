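# Unlit material for Control Rig control gizmos, giving them a flat color set by the Color vector parameter, which defaults to green.
# The color's RGB is divided by the EyeAdaptation value, cancelling auto exposure so the color keeps a steady brightness on screen.
# Ray-traced shadow casting is turned off.
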Material /ControlRig/Controls/ControlRigGizmoMaterial
graph {
 properties {bCastRayTracedShadows=false, pixelDepthOffsetMode=PDOM_Legacy, shadingModel=MSM_Unlit, usedShadingModels=Unlit}
 nodes [
  Constant() {r=1}
  Divide(A=VectorParameter() {parameterName=Color, defaultValue=<rgba 0 1 0.0898 0>}.RGB, B=EyeAdaptation().EyeAdaptation)
 ]
}
